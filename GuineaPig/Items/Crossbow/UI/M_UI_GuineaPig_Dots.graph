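# Translucent UI material that draws the GuineaPig crossbow reticle: a center dot and mirrored spread ticks.
# Ticks are SDF boxes on symmetry-mirrored, rotated UVs, with their offset driven by chained lerps for spread, ADS and fire pulses.
# The dot is an SDF circle with its own size and opacity.
# Smoothstepped SDFs provide glows, and named reroutes feed fills and glows into a blend scaled by an overall Opacity.
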Material /GuineaPig/Items/Crossbow/UI/M_UI_GuineaPig_Dots
graph {
 properties {blendMode=BLEND_Translucent, materialDomain=MD_UI, pixelDepthOffsetMode=PDOM_Legacy, usedShadingModels=Default Lit}
 nodes [
  n1 = NamedRerouteUsage()
  n2 = NamedRerouteUsage()
  n3 = NamedRerouteUsage()
  n4 = NamedRerouteUsage()
  Multiply(
   A=LinearInterpolate(A=Add(A=@n1, B=@n2), B=Add(A=@n3, B=@n4)) {constAlpha=0.4}
   B=ScalarParameter() {defaultValue=0.75, parameterName=Opacity}
  )
  n5 = NamedRerouteUsage()
  n6 = LinearInterpolate(
   A=LinearInterpolate(
    B=ScalarParameter() {defaultValue=-0.1, parameterName=SpreadTickOffset}
    Alpha=ScalarParameter() {group=Ticks, parameterName=X Offset, sliderMax=1, sortPriority=7}
   ) {constA=-0.305}
   Alpha=ScalarParameter() {parameterName=ADSOffset}
  ) {constB=-0.375}
  n7 = ScalarParameter() {parameterName=firePulse}
  n8 = AppendVector(
   A=LinearInterpolate(A=LinearInterpolate(A=@n6, Alpha=@n7) {constB=-0.21}, Alpha=ScalarParameter() {parameterName=ADSFirePulse}) {constB=-0.31}
   B=Constant()
  )
  n9 = Constant2Vector() {g=1, r=1}
  n10 = MaterialFunctionCall(
   UVs=MaterialFunctionCall(UVs=MaterialFunctionCall!(UV=@n5, Offset=@n8, Rotation, Scale=@n9, MirrorY) {materialFunction=/FortUILibrary/MaterialFunctions/Basics/MF_UI_Symmetry_2x1_DEPRECATED}.Result) {materialFunction=&CustomRotator}.Rotated Values
   Size=AppendVector(A=Constant() {r=0.095}, B=Constant() {r=0.005})
   GlowMax=Constant() {r=0.005}
  ) {materialFunction=&MF_UI_Box}
  NamedRerouteDeclaration(None=OneMinus(None=SmoothStep(Value=@n10.SDF) {constMax=0.015, constMin=0.001})) {name=spreadTicksGlow, nodeColor=<rgba 0 0 0 1>}
  n11 = MaterialFunctionCall(
   UVs=NamedRerouteUsage()
   Size=ScalarParameter() {defaultValue=0.0135, group=CenterDot, parameterName=CenterDotSize, sliderMax=0.1, sliderMin=0.0125}
   GlowMax=Constant() {r=0.005}
  ) {materialFunction=&MF_UI_Circle}
  NamedRerouteDeclaration(None=OneMinus(None=SmoothStep(Value=@n11.SDF) {constMax=0.015, constMin=0.001})) {name=centerDotGlow, nodeColor=<rgba 0 0 0 1>}
  NamedRerouteDeclaration(None=Multiply(A=@n11.Fill, B=ScalarParameter() {defaultValue=1, parameterName=CentralDotOpacity})) {name=CentralDot, nodeColor=<rgba 0 0 0 1>}
  NamedRerouteDeclaration(None=@n10.Fill) {name=SpreadTicks, nodeColor=<rgba 0 0 0 1>}
  NamedRerouteDeclaration(None=MaterialFunctionCall(
    UVs=MaterialFunctionCall() {materialFunction=&GetUserInterfaceUV}.Normalized UV
    Texture Scale=Constant2Vector() {g=1, r=1}
   ) {materialFunction=&ScaleUVsByCenter}.UVs) {name=UVs, nodeColor=<rgba 0 0 0 1>}
 ]
}
