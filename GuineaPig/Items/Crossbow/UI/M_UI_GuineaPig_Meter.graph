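# Translucent UI material for the GuineaPig crossbow's meter, drawing a progress arc with frame ring, pips and glows.
# SDF circles on a slightly oval UV form the rings, mirrored boxes form the pips, and a radial progress function driven by Progress masks the bar.
# Progress shifts the fill from orange to red, while Charged adds a sine-pulsed glow and alters the frame.
# meterSize and the Fire and ADS pulse parameters drive ring size and pip offset.
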Material /GuineaPig/Items/Crossbow/UI/M_UI_GuineaPig_Meter
graph {
 properties {blendMode=BLEND_Translucent, materialDomain=MD_UI, pixelDepthOffsetMode=PDOM_Legacy, usedShadingModels=Default Lit}
 nodes [
  n1 = NamedRerouteUsage()
  n2 = ScalarParameter() {defaultValue=0.75, parameterName=Opacity}
  n3 = NamedRerouteUsage()
  LinearInterpolate(A=Add(A=Multiply(A=@n1, B=@n2), B=NamedRerouteUsage()), B=Add(A=Multiply(A=@n3, B=@n2), B=NamedRerouteUsage())) {constAlpha=0.4}
  n4 = Add(A=ScalarParameter() {parameterName=Progress, sliderMax=1}) {constB=0.25}
  n5 = NamedRerouteUsage()
  n6 = Multiply(
   A=LinearInterpolate(A=Constant3Vector() {constant=<rgba 1 0.1254 0 1>}, B=Constant3Vector() {constant=<rgba 0.8784 0 0.0157 0>}, Alpha=Saturate(None=@n4))
   B=@n5
  )
  n7 = Multiply(A=NamedRerouteUsage()) {constB=0.625}
  NamedRerouteDeclaration(None=Multiply(A=Add(A=@n6, B=@n7), B=@n5)) {name=meterBarFill+Color, nodeColor=<rgba 0 0 0 1>}
  n8 = Constant() {desc=SDF Outline Thickness, r=-0.036}
  n9 = ScalarParameter() {parameterName=meterSize}
  n10 = LinearInterpolate(
   A=LinearInterpolate(A=LinearInterpolate(Alpha=@n9) {constA=-0.2, constB=-0.35, desc=ADS}, Alpha=ScalarParameter() {parameterName=Fire Pulse}) {constA=-0.2, constB=-0.1, desc=Normal Pulse}
   Alpha=ScalarParameter() {parameterName=ADS Pulse}
  ) {constA=-0.35, constB=-0.25, desc=ADS Pulse}
  n11 = Add(A=@n10, B=Constant() {desc=BarPlacement, r=0.5}) {constB=0.2}
  n12 = Add(A=@n11) {constB=0, desc=Meter Size}
  n13 = NamedRerouteUsage()
  n14 = Constant2Vector() {desc=Scale Size for Oval, g=1.075, r=1}
  n15 = Multiply(A=@n13, B=@n14)
  n16 = SmoothStep(Value=MaterialFunctionCall(UVs=@n15, Size=Subtract(A=@n12, B=@n8), OutlineThickness=@n8) {materialFunction=&MF_UI_Circle}.SDFOutline) {constMax=0.05, constMin=0.01}
  n17 = Time()
  n18 = ConstantBiasScale(None=Sine(None=Multiply(A=@n17) {constB=2}))
  n19 = ScalarParameter() {parameterName=Charged}
  NamedRerouteDeclaration(None=Multiply(A=OneMinus(None=@n16), B=Multiply(A=@n18, B=@n19) {constB=0})) {name=ChargedGlow, nodeColor=<rgba 0 0 0 1>}
  n20 = Constant() {r=180}
  n21 = MaterialFunctionCall!(
   UVs=MaterialFunctionCall!(UV=NamedRerouteUsage(), Offset=Constant2Vector() {g=0.011, r=-0.5}, Rotation, Scale, MirrorY) {materialFunction=/FortUILibrary/MaterialFunctions/Basics/MF_UI_Symmetry_2x1_DEPRECATED}.Result
   Progress=ScalarParameter() {parameterName=Progress, sliderMax=1}
   StartValue=@n20
   Range=Divide(A=@n20)
  ) {materialFunction=/FortUILibrary/MaterialFunctions/Basics/MF_UI_RadialProgress}
  n22 = NamedRerouteUsage()
  n23 = NamedRerouteUsage()
  NamedRerouteDeclaration(None=Multiply(A=Multiply(A=@n21.Result, B=@n22), B=OneMinus(None=@n23)) {constB=0.25}) {name=meterBarChargedGlow, nodeColor=<rgba 0 0 0 1>}
  n24 = Constant() {r=0.001}
  n25 = LinearInterpolate(Alpha=ScalarParameter() {defaultValue=1, parameterName=meterOuterPipsScale}) {constB=0.5}
  n26 = Constant() {r=0.5}
  n27 = MaterialFunctionCall(
   UVs=MaterialFunctionCall(UV=NamedRerouteUsage(), Offset=AppendVector(A=@n10, B=@n24), Scale=AppendVector(A=@n25, B=@n26)) {materialFunction=&MF_UI_Symmetry_2x1_DEPRECATED}.Result
   Size=Constant2Vector() {g=0.03, r=0.35}
   GlowMax=Constant() {r=0.015}
  ) {materialFunction=&MF_UI_Box}
  NamedRerouteDeclaration(None=@n27.Fill) {name=meterOuterPips, nodeColor=<rgba 0 0 0 1>}
  n28 = NamedRerouteUsage()
  n29 = Multiply(A=@n28, B=Constant2Vector() {g=2, r=1})
  n30 = Saturate(None=Floor(None=ComponentMask(None=@n29) {g=true}))
  NamedRerouteDeclaration(None=Multiply(A=@n28, B=NamedRerouteDeclaration(None=@n30) {name=uvHalfMask, nodeColor=<rgba 0.5938 0 1 1>})) {name=bottomHalfUV, nodeColor=<rgba 0 0 0 1>}
  n31 = ScalarParameter() {defaultValue=0.017, parameterName=outlineThickness}
  n32 = ScalarParameter() {parameterName=Charged}
  n33 = MaterialFunctionCall(
   UVs=Multiply(A=NamedRerouteUsage(), B=@n14)
   Size=Add(A=Subtract(A=@n11, B=@n31), B=LinearInterpolate(Alpha=@n32) {constA=-0.0075, constB=0.017})
   OutlineThickness=Multiply(A=@n31) {constB=2}
  ) {materialFunction=&MF_UI_Circle}
  NamedRerouteDeclaration(None=SmoothStep(Value=@n33.SDF) {constMax=0.01}) {name=circleMaskOuter, nodeColor=<rgba 0 0 0 1>}
  n34 = Constant() {r=0.005}
  n35 = MaterialFunctionCall(UVs=@n15, Size=@n12, OutlineThickness=Add(A=@n31, B=Constant() {r=0.035}), GlowMax=@n34) {materialFunction=&MF_UI_Circle}
  n36 = SmoothStep(Value=@n35.SDFOutline) {constMax=0.02}
  n37 = NamedRerouteUsage()
  NamedRerouteDeclaration(None=Multiply(A=OneMinus(None=@n36), B=@n37)) {name=meterBarFillGlow, nodeColor=<rgba 0 0 0 1>}
  n38 = MaterialFunctionCall(UVs=Multiply(A=@n13, B=@n14), Size=@n11, OutlineThickness=@n31, GlowMax=@n34) {materialFunction=&MF_UI_Circle}
  NamedRerouteDeclaration(None=@n38.Outline) {name=meterFrameRing, nodeColor=<rgba 0 0 0 1>}
  NamedRerouteDeclaration(None=Multiply(A=@n35.Outline, B=@n37)) {name=meterBarFill, nodeColor=<rgba 0 0 0 1>}
  n39 = NamedRerouteUsage()
  n40 = OneMinus(None=ScalarParameter() {parameterName=ADSChargedToggle})
  n41 = SmoothStep(Value=LinearInterpolate(A=NamedRerouteUsage(), B=NamedRerouteUsage(), Alpha=ScalarParameter() {parameterName=Charged})) {constMax=0.7, constMin=0.3}
  n42 = Multiply(A=Multiply(A=NamedRerouteUsage(), B=Ceil(None=@n40)), B=@n41)
  NamedRerouteDeclaration(None=Saturate(None=Max(A=@n39, B=@n42))) {name=meterFrame, nodeColor=<rgba 0 0 0 1>}
  n43 = Add(A=@n33.SDF, B=@n33.SDF)
  NamedRerouteDeclaration(None=OneMinus(None=SmoothStep(Value=@n43) {constMax=0.01})) {name=circleMaskInner, nodeColor=<rgba 0 0 0 1>}
  n44 = Floor(None=ComponentMask(None=Add(A=@n29) {constB=0.019}) {g=true})
  NamedRerouteDeclaration(None=Multiply(A=@n28, B=Saturate(None=@n44))) {name=bottomHalfUVOffset, nodeColor=<rgba 0 0 0 1>}
  n45 = OneMinus(None=NamedRerouteUsage())
  n46 = Multiply(A=NamedRerouteUsage(), B=@n21.Result)
  NamedRerouteDeclaration(None=NamedRerouteDeclaration(None=Multiply(A=@n45, B=@n46) {constB=0.25}) {name=meterBar, nodeColor=<rgba 0 0 0 1>}) {name=meterBarCharged, nodeColor=<rgba 0 0 0 1>}
  NamedRerouteDeclaration(None=OneMinus(None=SmoothStep(Value=@n27.SDF) {constMax=0.04})) {name=meterOuterPipsGlow, nodeColor=<rgba 0 0 0 1>}
  NamedRerouteDeclaration(None=OneMinus(None=SmoothStep(Value=@n38.SDFOutline) {constMax=0.02})) {name=meterFrameRingGlow, nodeColor=<rgba 0 0 0 1>}
  NamedRerouteDeclaration(None=MaterialFunctionCall(
    UVs=MaterialFunctionCall() {materialFunction=&GetUserInterfaceUV}.Normalized UV
    Texture Scale=Constant2Vector() {g=1, r=1}
   ) {materialFunction=&ScaleUVsByCenter}.UVs) {name=UVs, nodeColor=<rgba 0 0 0 1>}
  n47 = Multiply(A=@n41, B=NamedRerouteUsage())
  n48 = NamedRerouteUsage()
  NamedRerouteDeclaration(None=Saturate(None=Max(A=@n47, B=@n48))) {name=meterFrameGlow, nodeColor=<rgba 0 0 0 1>}
 ]
}
