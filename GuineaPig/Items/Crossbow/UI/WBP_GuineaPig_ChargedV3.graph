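# Charge reticle widget for the GuineaPig crossbow, drawing Dots and Meter material images over a FortChargedWeaponReticleExtensionWidget base.
# Charge updates set the Meter material's Progress and Charged parameters, with a delta gate playing the Charged animation only when the full-charge state flips.
# The weapon's targeting delegate plays the ADS scope animation forward or reverse, and firing selects a hip or ADS pulse.
# PreConstruct binds that delegate and enlarges both images on mobile.
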WidgetBlueprint /GuineaPig/Items/Crossbow/UI/WBP_GuineaPig_ChargedV3
animations [
 Charged {
  binding_count=1
  length_seconds=0.15
  object=&self:Charged
  sequence {bindings [Meter {
     id=#1
     possessed_class=Image
     tracks [{
       type=MovieSceneWidgetMaterialTrack
       sections [{
         type=MovieSceneParameterSection
         end_frame=3
         start_frame=0
         channels [Charged {type=MovieSceneScriptingFloatChannel, keys [[0,0],[1,1],[2,1]]}]
        }]
      }]
    }]}
 }
 ToggleADSScope {
  binding_count=2
  length_seconds=0.25
  object=&self:ToggleADSScope
  sequence {bindings [
    Dots {
     id=#2
     possessed_class=Image
     tracks [{
       type=MovieSceneWidgetMaterialTrack
       sections [{type=MovieSceneParameterSection, end_frame=5, start_frame=0, channels [ADSOffset {type=MovieSceneScriptingFloatChannel, keys [[0,0],[5,1]]}]}]
      }]
    }
    Meter {
     id=#3
     possessed_class=Image
     tracks [{
       type=MovieSceneWidgetMaterialTrack
       sections [{type=MovieSceneParameterSection, end_frame=5, start_frame=0, channels [meterSize {type=MovieSceneScriptingFloatChannel, keys [[0,0],[5,1]]}]}]
      }]
    }
   ]}
 }
 FirePulse {
  binding_count=2
  length_seconds=0.35
  object=&self:FirePulse
  sequence {bindings [
    Meter {
     id=#4
     possessed_class=Image
     tracks [{
       type=MovieSceneWidgetMaterialTrack
       sections [{
         type=MovieSceneParameterSection
         end_frame=7
         start_frame=0
         channels [Fire Pulse {type=MovieSceneScriptingFloatChannel, keys [[0,0],[1,1],[7,0]]}]
        }]
      }]
    }
    Dots {
     id=#5
     possessed_class=Image
     tracks [{
       type=MovieSceneWidgetMaterialTrack
       sections [{
         type=MovieSceneParameterSection
         end_frame=7
         start_frame=0
         channels [firePulse {type=MovieSceneScriptingFloatChannel, keys [[0,0],[1,1],[7,0]]}]
        }]
      }]
    }
   ]}
 }
 ADSPulse {
  binding_count=2
  length_seconds=0.35
  object=&self:ADSPulse
  sequence {bindings [
    Meter {
     id=#6
     possessed_class=Image
     tracks [{
       type=MovieSceneWidgetMaterialTrack
       sections [{
         type=MovieSceneParameterSection
         end_frame=7
         start_frame=0
         channels [ADS Pulse {type=MovieSceneScriptingFloatChannel, keys [[0,0],[1,1],[7,0]]}]
        }]
      }]
    }
    Dots {
     id=#7
     possessed_class=Image
     tracks [{
       type=MovieSceneWidgetMaterialTrack
       sections [{
         type=MovieSceneParameterSection
         end_frame=7
         start_frame=0
         channels [ADSFirePulse {type=MovieSceneScriptingFloatChannel, keys [[0,0],[1,1],[7,0]]}]
        }]
      }]
    }
   ]}
 }
]
graphs [
 EventGraph {nodes [
   n1 = ADSToggle!() -> (OutputDelegate, then, bIsTargeting:bool=~) {is=title K2Node_CustomEvent}
   > n2 = Set ADS?(ADS?:bool=@n1.bIsTargeting) -> (Output_Get:bool)
   > n3 = Branch(Condition=@n2)
   n4 = Get ToggleADSScope() -> (ToggleADSScope:Widget Animation Object Reference)
   > PlayAnimationForward(InAnimation=@n4)
   n5 = Branch(execute=@n3.else, Condition=IsAnimationPlaying(InAnimation=Get ADSPulse() -> (ADSPulse:Widget Animation Object Reference)))
   > PlayAnimation(InAnimation=@n4, StartAtTime=GetEndTime(self=@n4), PlayMode=Reverse)
   PlayAnimationReverse(execute=@n5.else, InAnimation=@n4)
   Event PreConstruct()
   > Bind Event to On Targeting Changed(
    self:Fort Weapon Object Reference=GetEquippedWeapon!(self:Fort Weapon Reticle Extension Widget Base Object Reference) -> (ReturnValue:Fort Weapon Object Reference)
    Delegate=Create Event()
   ) {is=title K2Node_AddDelegate}
   > SetMobileStyle() {is=title K2Node_CallFunction}
   n6 = Event OnUpdateChargeDisplay!() -> (OutputDelegate, then, ChargePercent:float) {is=title K2Node_Event}
   > n7 = GetDynamicMaterial(self=Get Meter() -> (Meter:Image Object Reference))
   > SetScalarParameterValue(self=@n7, ParameterName=Progress, Value=@n6.ChargePercent)
   n8 = "float >= float"(A=@n6.ChargePercent, B=1)
   > SetScalarParameterValue(self=@n7, ParameterName=Charged, Value="To Float (Boolean)"(InBool=@n8))
   > n9 = Delta Gate Wildcard!(execute:Exec, Variable:"Boolean (by ref)"=Get Charged?() -> (Charged?:bool), New Value:bool=@n8) -> (Success:Exec, Updated Variable:"Boolean (by ref)", Previous Variable:bool=~) {is=title K2Node_MacroInstance}
   > n10 = Branch(Condition=@n9)
   n11 = Get Charged() -> (Charged:Widget Animation Object Reference)
   > PlayAnimationForward(InAnimation=@n11)
   PlayAnimationReverse(execute=@n10.else, InAnimation=@n11, PlaybackSpeed=2)
   Event OnWeaponSingleFire() {is=title K2Node_Event}
   n12 = Get ADS?() -> (ADS?:bool)
   > PlayAnimation(InAnimation=Select(
     Option 0:Widget Animation Object Reference=Get FirePulse() -> (FirePulse:Widget Animation Object Reference)
     Option 1:Widget Animation Object Reference=Get ADSPulse()
     Index=@n12
    ) -> (ReturnValue:Widget Animation Object Reference))
   n13 = Get ToggleADSScope()
   > Branch(Condition=AND Boolean(A=IsAnimationPlaying(InAnimation=@n13), B=NOT Boolean(A=@n12)))
   > PlayAnimation(InAnimation=@n13, StartAtTime=GetEndTime(self=@n13), PlayMode=Reverse)
  ]}
 SetMobileStyle {nodes [
   SetMobileStyle() {is=title K2Node_FunctionEntry}
   > Branch(Condition=IsMobileGame())
   > SetRenderScale!(execute, self:Widget Object Reference=[Get Meter(), Get Dots() -> (Dots:Image Object Reference)], Scale:Vector 2D Structure=(X=1.5,Y=1.5)) -> (then)
  ]}
]
variables [
 ADS? {type=boolean}
 Charged? {type=boolean}
 RetMat {type=MaterialInstanceDynamic}
]
widgets {
 info {
  inheritedWidgetCount=1
  parentClass=/Script/FortniteUI.FortChargedWeaponReticleExtensionWidget
  rootWidgetClass=&Overlay
  widgetCount=4
 }
 widgets [
  Overlay_0 {widgetClassPath=&Overlay, widget=Overlay_0}
  Dots {
   widgetClassPath=&Image
   bIsVariable=true
   parent=Overlay_0
   widget=Dots
   properties {brush {
     bIsDynamicallyLoaded=false
     drawAs=Image
     imageType=FullColor
     mirroring=NoMirror
     resourceName=None
     resourceObject=&/MI_UI_GuineaPig_Dots
     tiling=NoTile
     imageSize=<256 256>
     margin=<ltrb 0 0 0 0>
     outlineSettings {
      bUseBrushTransparency=false
      roundingType=HalfHeightRadius
      width=0
      color {colorUseRule=UseColor_Specified, specifiedColor=<rgba 0 0 0 0>}
      cornerRadii=<0 0 0 0>
     }
     tintColor {colorUseRule=UseColor_Specified, specifiedColor=<rgba 1 1 1 1>}
     uVRegion {bIsValid=false, max=<0 0>, min=<0 0>}
    }}
   slot {class=OverlaySlot, properties {horizontalAlignment=HAlign_Center, verticalAlignment=VAlign_Center}}
  }
  Meter {
   widgetClassPath=&Image
   bIsVariable=true
   parent=Overlay_0
   widget=Meter
   properties {
    brush {
     bIsDynamicallyLoaded=false
     drawAs=Image
     imageType=FullColor
     mirroring=NoMirror
     resourceName=None
     resourceObject=&/MI_UI_GuineaPig_Meter
     tiling=NoTile
     imageSize=<165 165>
     margin=<ltrb 0 0 0 0>
     outlineSettings {
      bUseBrushTransparency=false
      roundingType=HalfHeightRadius
      width=0
      color {colorUseRule=UseColor_Specified, specifiedColor=<rgba 0 0 0 0>}
      cornerRadii=<0 0 0 0>
     }
     tintColor {colorUseRule=UseColor_Specified, specifiedColor=<rgba 1 1 1 1>}
     uVRegion {bIsValid=false, max=<0 0>, min=<0 0>}
    }
    renderTransform {angle=0, scale=<1 1>, shear=<0 0>, translation=<0 10>}
   }
   slot {class=OverlaySlot, properties {horizontalAlignment=HAlign_Center, verticalAlignment=VAlign_Center}}
  }
  TutorialHighlight {widgetClassPath=&TutorialHighlight}
 ]
}
